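; Day sequence volume for the Hades area that fades the Hades cloud look in and out as the blend target enters or leaves.
; Overlap begin plays a timeline and overlap end reverses it, writing each update to the HadesActive scalar in the global FortniteMaterialParameters collection.
; A timeline replaces distance-based blending, which feels laggy because server player position updates step.
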(event EventBeginPlay
  (|CallParentFunction) ; This node is disabled and will not be called.\nDrag off pins to build functionality.
)

(event Collision|EventActorBeginOverlap
  (params (OtherActor :type Actor%20Object%20Reference)))

(event EventTick
  (params (DeltaSeconds :type Float%20%28single-precision%29)))

(event EventPostTODMCached
  ;[1 1 1] Hades Clouds - Doing it with a timeline because the stepping due to server player position updates makes 'distance' method feel laggy
  (|CallParentFunction))

(event Custom|OnVolumeBlendTargetOverlapBegin_Event
  (params (OverlappingActor :type Actor%20Object%20Reference))
  ;[1 1 1] Hades Clouds - Doing it with a timeline because the stepping due to server player position updates makes 'distance' method feel laggy
  (bind (_direction _hadesactive) (Timeline|Timeline Play :NewTime (Variables|Default|Gettime))
    (:Update
      (Rendering|Material|SetScalarParameterValue "/Game/Packages/Fortress_SharedMaterials/GlobalMaterialParameters/FortniteMaterialParameters.FortniteMaterialParameters" "HadesActive" _hadesactive :!class /Script/Engine.KismetMaterialLibrary))))

(event Custom|OnVolumeBlendTargetOverlapEnd_Event
  (params (OverlappingActor :type Actor%20Object%20Reference))
  ;[1 1 1] Hades Clouds - Doing it with a timeline because the stepping due to server player position updates makes 'distance' method feel laggy
  (Timeline|Timeline Reverse))

(fn ConstructionScript
  (|CallParentFunction))
